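; Controller requirement that passes only when the Living World Manager has enough budget for a configured event.
; It gets the Athena game state from the test subject actor and finds its FortAthenaLivingWorldManager component.
; It queries that manager's event budget using the EventRowHandle variable and the context's other actor, then compares it against RequiredBudget with greater-or-equal.
; An invalid subject actor, game state or manager fails the requirement.
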(fn IsRequirementMetInternal
  (params (RequestContext :type Controller%20Requirement%20Test%20Context%20Structure%20%28by%20ref%29))
  (bind (_testsubjectactor _testsubjectcontroller _otheractor) (Utilities|Struct|BreakControllerRequirementTestContext RequestContext))
  (bind _returnvalue (Fortnite|Athena|GameState|GetGameStateAthenafromContext _testsubjectactor))
  (bind _returnvalue_2 (Actor|GetComponentbyClass _returnvalue "/Script/LagerRuntime.FortAthenaLivingWorldManager"))
  (Utilities|IsValid _testsubjectactor
    (:Is%20Valid
      (Utilities|IsValid _returnvalue
        (:Is%20Valid
          (Utilities|IsValid _returnvalue_2
            (:Is%20Valid
              (return (>= (Class|FortAthenaLivingWorldManager|QueryEventBudget _returnvalue_2 (Variables|Default|GetEventRowHandle) _otheractor) (Variables|Default|GetRequiredBudget))))
            (:Is%20Not%20Valid
              (return false))))
        (:Is%20Not%20Valid
          (return false))))
    (:Is%20Not%20Valid
      (return false))))
